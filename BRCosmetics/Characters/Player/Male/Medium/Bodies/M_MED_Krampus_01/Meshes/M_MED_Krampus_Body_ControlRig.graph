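# Control Rig for the Krampus medium male body that makes the extra wolf leg bones follow the standard biped legs.
# Per side, wolf_thigh, wolf_calf and wolf_ball follow thigh, calf and foot in turn; wolf_foot is not driven directly.
# Each ApplyFK overrides the wolf joint in its biped joint's space, using the relative offset between their initial global transforms.
# Two Control units and several BeginExecution nodes are unconnected.
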ControlRigBlueprint /BRCosmetics/Characters/Player/Male/Medium/Bodies/M_MED_Krampus_01/Meshes/M_MED_Krampus_Body_ControlRig
graphs [
 RigVMFunctionLibrary {links [], nodes []}
 RigVMModel {nodes [
   n1 = &RigUnit_BeginExecution()
   n2 = /Script/ControlRig.RigUnit_GetJointTransform(
    BaseJoint=None
    BaseTransform=(Rotation=(X=0,Y=0,Z=0,W=1),Translation=(X=0,Y=0,Z=0),Scale3D=(X=1,Y=1,Z=1))
    Joint=thigh_l
    TransformSpace=GlobalSpace
    Type=Initial
    ExecutePin=@n1.ExecutePin
   ) {is=struct RigVMUnitNode}
   n3 = &RigUnit_BeginExecution()
   n4 = &RigUnit_GetJointTransform(
    BaseJoint=None
    BaseTransform=(Rotation=(X=0,Y=0,Z=0,W=1),Translation=(X=0,Y=0,Z=0),Scale3D=(X=1,Y=1,Z=1))
    Joint=wolf_thigh_l
    TransformSpace=GlobalSpace
    Type=Initial
    ExecutePin=@n3.ExecutePin
   )
   n5 = &RigUnit_BeginExecution()
   n6 = /Script/ControlRig.RigUnit_ApplyFK(
    ApplyTransformMode=Override
    ApplyTransformSpace=BaseJoint
    BaseJoint=thigh_l
    BaseTransform=(Rotation=(X=0,Y=0,Z=0,W=1),Translation=(X=0,Y=0,Z=0),Scale3D=(X=1,Y=1,Z=1))
    Filter=(TranslationFilter=(bX=True,bY=True,bZ=True),RotationFilter=(bX=True,bY=True,bZ=True),ScaleFilter=(bX=True,bY=True,bZ=True))
    Joint=wolf_thigh_l
    ExecutePin=@n5.ExecutePin
    Transform=/Script/ControlRig.RigUnit_GetRelativeTransform(Argument1=@n2.Output, Argument0=@n4.Output) {is=struct RigVMUnitNode}.Result
   ) {is=struct RigVMUnitNode}
   n7 = &RigUnit_GetJointTransform(
    BaseJoint=None
    BaseTransform=(Rotation=(X=0,Y=0,Z=0,W=1),Translation=(X=0,Y=0,Z=0),Scale3D=(X=1,Y=1,Z=1))
    Joint=calf_l
    TransformSpace=GlobalSpace
    Type=Initial
    ExecutePin=@n23.ExecutePin
   )
   n8 = &RigUnit_GetJointTransform(
    BaseJoint=None
    BaseTransform=(Rotation=(X=0,Y=0,Z=0,W=1),Translation=(X=0,Y=0,Z=0),Scale3D=(X=1,Y=1,Z=1))
    Joint=wolf_calf_l
    TransformSpace=GlobalSpace
    Type=Initial
    ExecutePin=@n24.ExecutePin
   )
   n9 = &RigUnit_ApplyFK(
    ApplyTransformMode=Override
    ApplyTransformSpace=BaseJoint
    BaseJoint=calf_l
    BaseTransform=(Rotation=(X=0,Y=0,Z=0,W=1),Translation=(X=0,Y=0,Z=0),Scale3D=(X=1,Y=1,Z=1))
    Filter=(TranslationFilter=(bX=True,bY=True,bZ=True),RotationFilter=(bX=True,bY=True,bZ=True),ScaleFilter=(bX=True,bY=True,bZ=True))
    Joint=wolf_calf_l
    ExecutePin=@n6.ExecutePin
    Transform=&RigUnit_GetRelativeTransform(Argument1=@n7.Output, Argument0=@n8.Output).Result
   )
   n10 = &RigUnit_GetJointTransform(
    BaseJoint=None
    BaseTransform=(Rotation=(X=0,Y=0,Z=0,W=1),Translation=(X=0,Y=0,Z=0),Scale3D=(X=1,Y=1,Z=1))
    Joint=foot_l
    TransformSpace=GlobalSpace
    Type=Initial
    ExecutePin=@n29.ExecutePin
   )
   n11 = &RigUnit_GetJointTransform(
    BaseJoint=None
    BaseTransform=(Rotation=(X=0,Y=0,Z=0,W=1),Translation=(X=0,Y=0,Z=0),Scale3D=(X=1,Y=1,Z=1))
    Joint=wolf_ball_l
    TransformSpace=GlobalSpace
    Type=Initial
    ExecutePin=@n30.ExecutePin
   )
   &RigUnit_ApplyFK(
    ApplyTransformMode=Override
    ApplyTransformSpace=BaseJoint
    BaseJoint=foot_l
    BaseTransform=(Rotation=(X=0,Y=0,Z=0,W=1),Translation=(X=0,Y=0,Z=0),Scale3D=(X=1,Y=1,Z=1))
    Filter=(TranslationFilter=(bX=True,bY=True,bZ=True),RotationFilter=(bX=True,bY=True,bZ=True),ScaleFilter=(bX=True,bY=True,bZ=True))
    Joint=wolf_ball_l
    ExecutePin=@n9.ExecutePin
    Transform=&RigUnit_GetRelativeTransform(Argument1=@n10.Output, Argument0=@n11.Output).Result
   )
   n12 = &RigUnit_BeginExecution()
   n13 = &RigUnit_GetJointTransform(
    BaseJoint=None
    BaseTransform=(Rotation=(X=0,Y=0,Z=0,W=1),Translation=(X=0,Y=0,Z=0),Scale3D=(X=1,Y=1,Z=1))
    Joint=thigh_r
    TransformSpace=GlobalSpace
    Type=Initial
    ExecutePin=@n12.ExecutePin
   )
   n14 = &RigUnit_BeginExecution()
   n15 = &RigUnit_GetJointTransform(
    BaseJoint=None
    BaseTransform=(Rotation=(X=0,Y=0,Z=0,W=1),Translation=(X=0,Y=0,Z=0),Scale3D=(X=1,Y=1,Z=1))
    Joint=wolf_thigh_r
    TransformSpace=GlobalSpace
    Type=Initial
    ExecutePin=@n14.ExecutePin
   )
   n16 = &RigUnit_BeginExecution()
   n17 = &RigUnit_ApplyFK(
    ApplyTransformMode=Override
    ApplyTransformSpace=BaseJoint
    BaseJoint=thigh_r
    BaseTransform=(Rotation=(X=0,Y=0,Z=0,W=1),Translation=(X=0,Y=0,Z=0),Scale3D=(X=1,Y=1,Z=1))
    Filter=(TranslationFilter=(bX=True,bY=True,bZ=True),RotationFilter=(bX=True,bY=True,bZ=True),ScaleFilter=(bX=True,bY=True,bZ=True))
    Joint=wolf_thigh_r
    ExecutePin=@n16.ExecutePin
    Transform=&RigUnit_GetRelativeTransform(Argument1=@n13.Output, Argument0=@n15.Output).Result
   )
   n18 = &RigUnit_GetJointTransform(
    BaseJoint=None
    BaseTransform=(Rotation=(X=0,Y=0,Z=0,W=1),Translation=(X=0,Y=0,Z=0),Scale3D=(X=1,Y=1,Z=1))
    Joint=calf_r
    TransformSpace=GlobalSpace
    Type=Initial
    ExecutePin=@n25.ExecutePin
   )
   n19 = &RigUnit_GetJointTransform(
    BaseJoint=None
    BaseTransform=(Rotation=(X=0,Y=0,Z=0,W=1),Translation=(X=0,Y=0,Z=0),Scale3D=(X=1,Y=1,Z=1))
    Joint=wolf_calf_r
    TransformSpace=GlobalSpace
    Type=Initial
    ExecutePin=@n26.ExecutePin
   )
   n20 = &RigUnit_ApplyFK(
    ApplyTransformMode=Override
    ApplyTransformSpace=BaseJoint
    BaseJoint=calf_r
    BaseTransform=(Rotation=(X=0,Y=0,Z=0,W=1),Translation=(X=0,Y=0,Z=0),Scale3D=(X=1,Y=1,Z=1))
    Filter=(TranslationFilter=(bX=True,bY=True,bZ=True),RotationFilter=(bX=True,bY=True,bZ=True),ScaleFilter=(bX=True,bY=True,bZ=True))
    Joint=wolf_calf_r
    ExecutePin=@n17.ExecutePin
    Transform=&RigUnit_GetRelativeTransform(Argument1=@n18.Output, Argument0=@n19.Output).Result
   )
   n21 = &RigUnit_GetJointTransform(
    BaseJoint=None
    BaseTransform=(Rotation=(X=0,Y=0,Z=0,W=1),Translation=(X=0,Y=0,Z=0),Scale3D=(X=1,Y=1,Z=1))
    Joint=foot_r
    TransformSpace=GlobalSpace
    Type=Initial
    ExecutePin=@n27.ExecutePin
   )
   n22 = &RigUnit_GetJointTransform(
    BaseJoint=None
    BaseTransform=(Rotation=(X=0,Y=0,Z=0,W=1),Translation=(X=0,Y=0,Z=0),Scale3D=(X=1,Y=1,Z=1))
    Joint=wolf_ball_r
    TransformSpace=GlobalSpace
    Type=Initial
    ExecutePin=@n28.ExecutePin
   )
   &RigUnit_ApplyFK(
    ApplyTransformMode=Override
    ApplyTransformSpace=BaseJoint
    BaseJoint=foot_r
    BaseTransform=(Rotation=(X=0,Y=0,Z=0,W=1),Translation=(X=0,Y=0,Z=0),Scale3D=(X=1,Y=1,Z=1))
    Filter=(TranslationFilter=(bX=True,bY=True,bZ=True),RotationFilter=(bX=True,bY=True,bZ=True),ScaleFilter=(bX=True,bY=True,bZ=True))
    Joint=wolf_ball_r
    ExecutePin=@n20.ExecutePin
    Transform=&RigUnit_GetRelativeTransform(Argument1=@n21.Output, Argument0=@n22.Output).Result
   )
   n23 = &RigUnit_BeginExecution()
   n24 = &RigUnit_BeginExecution()
   n25 = &RigUnit_BeginExecution()
   n26 = &RigUnit_BeginExecution()
   n27 = &RigUnit_BeginExecution()
   n28 = &RigUnit_BeginExecution()
   n29 = &RigUnit_BeginExecution()
   n30 = &RigUnit_BeginExecution()
   &RigUnit_BeginExecution()
   &RigUnit_BeginExecution()
   &RigUnit_BeginExecution()
   &RigUnit_BeginExecution()
   &RigUnit_BeginExecution()
   &RigUnit_BeginExecution()
   &RigUnit_BeginExecution()
   &RigUnit_BeginExecution()
   &RigUnit_BeginExecution()
   &RigUnit_BeginExecution()
   &RigUnit_BeginExecution()
   &RigUnit_BeginExecution()
   &RigUnit_BeginExecution()
   &RigUnit_BeginExecution()
   /Script/ControlRig.RigUnit_Control(
    Base=(Rotation=(X=-0.5214,Y=0.4801,Z=0.4735,W=-0.5229),Translation=(X=-14.5112,Y=-5.1023,Z=12.1062),Scale3D=(X=1,Y=1,Z=1))
    Filter=(TranslationFilter=(bX=True,bY=True,bZ=True),RotationFilter=(bX=True,bY=True,bZ=True),ScaleFilter=(bX=True,bY=True,bZ=True))
    InitTransform=(Rotation=(X=0,Y=0,Z=0,W=1),Translation=(X=0,Y=0,Z=0),Scale3D=(X=1,Y=1,Z=1))
   ) {is=struct RigVMUnitNode}
   &RigUnit_Control(
    Base=(Rotation=(X=0.5229,Y=0.4735,Z=-0.4801,W=-0.5214),Translation=(X=14.5112,Y=-5.1023,Z=12.1062),Scale3D=(X=1,Y=1,Z=1))
    Filter=(TranslationFilter=(bX=True,bY=True,bZ=True),RotationFilter=(bX=True,bY=True,bZ=True),ScaleFilter=(bX=True,bY=True,bZ=True))
    InitTransform=(Rotation=(X=0,Y=0,Z=0,W=1),Translation=(X=0,Y=0,Z=0),Scale3D=(X=1,Y=1,Z=1))
   )
  ]}
]
hierarchy [
 root {loc=<0 0 0>, rot=<0 0 0 1>}
 attach {parent=root, loc=<0 0 0>, rot=<0 0 0 1>}
 pelvis {parent=root, loc=<0 -0.7071 93.1191>, rot=<0.0127 0.707 -0.0127 -0.707>}
 spine_01 {parent=pelvis, loc=<6.7494 -0.6388 0>, rot=<0 0 0.034 -0.9994>}
 spine_02 {parent=spine_01, loc=<7.2518 -0.035 0>, rot=<0 0 0.0238 -0.9997>}
 spine_03 {parent=spine_02, loc=<6.0104 0 0>, rot=<0 0 0.0149 -0.9999>}
 spine_04 {parent=spine_03, loc=<8.1024 0 0>, rot=<0 0 -0.0132 -0.9999>}
 spine_05 {parent=spine_04, loc=<10.0146 0 0>, rot=<0 0 -0.0015 -1>}
 clavicle_l {parent=spine_05, loc=<4.3575 -0.1466 -2.3223>, rot=<0.0465 -0.7289 -0.009 -0.6829>}
 upperarm_l {parent=clavicle_l, loc=<15.5111 0 0>, rot=<0.0305 -0.536 -0.0057 -0.8437>}
 lowerarm_l {parent=upperarm_l, loc=<26.6514 0 0>, rot=<0.0391 -0.0099 -0.2453 -0.9686>}
 hand_l {parent=lowerarm_l, loc=<22.3491 0 0>, rot=<-0.7081 -0.0507 0.0975 -0.6975>}
 index_metacarpal_l {parent=hand_l, loc=<1.6535 0.0183 -2.96>, rot=<0.0158 -0.0124 -0.0173 -0.9996>}
 index_01_l {parent=index_metacarpal_l, loc=<8.3008 0.0357 -0.0524>, rot=<-0.0044 0.0485 0.0924 -0.9945>}
 index_02_l {parent=index_01_l, loc=<3.3587 -0.3004 0.0452>, rot=<-0.002 0.0027 0.3225 -0.9466>}
 index_03_l {parent=index_02_l, loc=<2.5578 -0.123 -0.0252>, rot=<-0.002 0.0067 0.2558 -0.9667>}
 middle_metacarpal_l {parent=hand_l, loc=<2.1856 -0.1725 -0.5984>, rot=<0.006 0.0168 -0.0329 -0.9993>}
 middle_01_l {parent=middle_metacarpal_l, loc=<7.3275 -0.2447 0.0129>, rot=<-0.0208 0.0042 0.0858 -0.9961>}
 middle_02_l {parent=middle_01_l, loc=<4.0123 0.1401 0.1175>, rot=<-0.0218 0.0258 0.3512 -0.9357>}
 middle_03_l {parent=middle_02_l, loc=<2.6077 0.0534 0.1373>, rot=<0.0082 0.0188 0.1708 -0.9851>}
 pinky_metacarpal_l {parent=hand_l, loc=<1.7616 -0.9753 3.3814>, rot=<-0.0397 0.0814 -0.069 -0.9935>}
 pinky_01_l {parent=pinky_metacarpal_l, loc=<7.1709 -0.4831 -0.1094>, rot=<-0.0654 -0.0035 0.2242 -0.9723>}
 pinky_02_l {parent=pinky_01_l, loc=<2.8274 0.0893 -0.1043>, rot=<0.0037 0.0011 0.2854 -0.9584>}
 pinky_03_l {parent=pinky_02_l, loc=<2.2046 0.0307 -0.0078>, rot=<-0.0002 0.0098 0.1445 -0.9895>}
 ring_metacarpal_l {parent=hand_l, loc=<2.2074 -0.3625 1.3638>, rot=<-0.0284 0.0722 -0.0458 -0.9959>}
 ring_01_l {parent=ring_metacarpal_l, loc=<7.0295 -0.6197 -0.1446>, rot=<-0.0489 -0.0252 0.1892 -0.9804>}
 ring_02_l {parent=ring_01_l, loc=<3.6174 0.4647 -0.1027>, rot=<0 0 0.2806 -0.9598>}
 ring_03_l {parent=ring_02_l, loc=<2.4152 0.1141 0.0199>, rot=<0 0 0.1189 -0.9929>}
 thumb_01_l {parent=hand_l, loc=<1.3118 -0.3191 -2.459>, rot=<0.6546 -0.4042 0.0664 -0.6354>}
 thumb_02_l {parent=thumb_01_l, loc=<5.9336 -0.0448 0.0401>, rot=<-0.0037 0.0213 0.298 -0.9543>}
 thumb_03_l {parent=thumb_02_l, loc=<2.2223 0.3682 -0.0907>, rot=<0.0118 -0.013 0.0066 -0.9998>}
 weapon_l {parent=hand_l, loc=<-0.1974 -3.4111 0.3811>, rot=<0 0 0 -1>}
 dyn_fur_l {parent=hand_l, loc=<-0.0689 0.3273 -0.3048>, rot=<-0.5621 -0.5887 0.3223 0.4834>}
 lowerarm_twist_01_l {parent=lowerarm_l, loc=<17.8549 0 0>, rot=<0 0 0 1>}
 lowerarm_twist_02_l {parent=lowerarm_l, loc=<9.6199 0 0>, rot=<0 0 0 1>}
 upperarm_twist_01_l {parent=upperarm_l, loc=<0.0041 0 0>, rot=<0 0 0 1>}
 upperarm_twist_02_l {parent=upperarm_l, loc=<13.7941 0 0>, rot=<0 0 0 1>}
 dyn_shoulder_sleeve_01_l {parent=upperarm_l, loc=<2.0599 0 0.0002>, rot=<0 0 0 1>}
 dyn_shoulder_sleeve_02_l {parent=dyn_shoulder_sleeve_01_l, loc=<-2.1191 0 5.383>, rot=<0 0 0 1>}
 dyn_shoulder_chain_l_01 {parent=clavicle_l, loc=<10.4279 -10.8553 0.5877>, rot=<-0.3895 -0.3686 0.5684 -0.6239>}
 dyn_shoulder_chain_l_02 {parent=dyn_shoulder_chain_l_01, loc=<4.0725 0 0>, rot=<0 -0.0872 0 -0.9962>}
 dyn_shoulder_chain_l_03 {parent=dyn_shoulder_chain_l_02, loc=<2.1807 0 0.0516>, rot=<0 -0.1305 0 -0.9914>}
 clavicle_r {parent=spine_05, loc=<4.3575 -0.1466 2.3223>, rot=<0.7289 0.0465 -0.6829 0.009>}
 upperarm_r {parent=clavicle_r, loc=<-15.5111 0 0>, rot=<0.0305 -0.536 -0.0057 -0.8437>}
 lowerarm_r {parent=upperarm_r, loc=<-26.6514 0 0>, rot=<0.0391 -0.0099 -0.2453 -0.9686>}
 hand_r {parent=lowerarm_r, loc=<-22.3492 0 0>, rot=<-0.7081 -0.0507 0.0975 -0.6975>}
 index_metacarpal_r {parent=hand_r, loc=<-1.6534 -0.0181 2.9601>, rot=<0.0158 -0.0124 -0.0173 -0.9996>}
 index_01_r {parent=index_metacarpal_r, loc=<-8.3008 -0.0357 0.0524>, rot=<-0.0044 0.0485 0.0924 -0.9945>}
 index_02_r {parent=index_01_r, loc=<-3.3589 0.3003 -0.0453>, rot=<-0.002 0.0027 0.3225 -0.9466>}
 index_03_r {parent=index_02_r, loc=<-2.5578 0.1231 0.0251>, rot=<-0.002 0.0067 0.2558 -0.9667>}
 middle_metacarpal_r {parent=hand_r, loc=<-2.1856 0.1726 0.5984>, rot=<0.006 0.0168 -0.0329 -0.9993>}
 middle_01_r {parent=middle_metacarpal_r, loc=<-7.3273 0.2447 -0.0129>, rot=<-0.0208 0.0042 0.0858 -0.9961>}
 middle_02_r {parent=middle_01_r, loc=<-4.0142 -0.1411 -0.07>, rot=<-0.0217 0.0259 0.3512 -0.9357>}
 middle_03_r {parent=middle_02_r, loc=<-2.6074 -0.0534 -0.1312>, rot=<0.0083 0.0188 0.1708 -0.9851>}
 pinky_metacarpal_r {parent=hand_r, loc=<-1.7617 0.9753 -3.3814>, rot=<-0.0397 0.0814 -0.069 -0.9935>}
 pinky_01_r {parent=pinky_metacarpal_r, loc=<-7.1707 0.4832 0.1094>, rot=<-0.0654 -0.0035 0.2242 -0.9723>}
 pinky_02_r {parent=pinky_01_r, loc=<-2.8274 -0.0892 0.1043>, rot=<0.0037 0.0011 0.2854 -0.9584>}
 pinky_03_r {parent=pinky_02_r, loc=<-2.2045 -0.031 0.0078>, rot=<-0.0002 0.0098 0.1445 -0.9895>}
 ring_metacarpal_r {parent=hand_r, loc=<-2.2073 0.3627 -1.3638>, rot=<-0.0284 0.0722 -0.0458 -0.9959>}
 ring_01_r {parent=ring_metacarpal_r, loc=<-7.0297 0.6197 0.1446>, rot=<-0.0489 -0.0252 0.1892 -0.9804>}
 ring_02_r {parent=ring_01_r, loc=<-3.6171 -0.4648 0.1027>, rot=<0 0 0.2806 -0.9598>}
 ring_03_r {parent=ring_02_r, loc=<-2.4153 -0.1141 -0.0199>, rot=<0 0 0.1189 -0.9929>}
 thumb_01_r {parent=hand_r, loc=<-1.3119 0.3192 2.459>, rot=<0.6546 -0.4042 0.0664 -0.6354>}
 thumb_02_r {parent=thumb_01_r, loc=<-5.9335 0.0447 -0.0401>, rot=<-0.0037 0.0213 0.298 -0.9543>}
 thumb_03_r {parent=thumb_02_r, loc=<-2.2223 -0.3682 0.0907>, rot=<0.0118 -0.013 0.0066 -0.9998>}
 weapon_r {parent=hand_r, loc=<0.1975 3.4112 -0.3811>, rot=<0 0 0 -1>}
 dyn_fur_r {parent=hand_r, loc=<0.0689 -0.3272 0.3047>, rot=<-0.4834 -0.3223 -0.5887 -0.5621>}
 lowerarm_twist_01_r {parent=lowerarm_r, loc=<-17.8549 0 0>, rot=<0 0 0 1>}
 lowerarm_twist_02_r {parent=lowerarm_r, loc=<-9.6199 0 0>, rot=<0 0 0 1>}
 upperarm_twist_01_r {parent=upperarm_r, loc=<-0.0041 0 0>, rot=<0 0 0 1>}
 upperarm_twist_02_r {parent=upperarm_r, loc=<-13.7942 0 0>, rot=<0 0 0 1>}
 dyn_shoulder_sleeve_01_r {parent=upperarm_r, loc=<-2.0604 0 0>, rot=<0 0 0 1>}
 dyn_shoulder_sleeve_02_r {parent=dyn_shoulder_sleeve_01_r, loc=<2.1194 0 -5.3831>, rot=<0 0 0 1>}
 dyn_shoulder_chain_r_01 {parent=clavicle_r, loc=<-10.4316 10.8039 -0.4931>, rot=<-0.5684 -0.6239 -0.3895 0.3686>}
 dyn_shoulder_chain_r_02 {parent=dyn_shoulder_chain_r_01, loc=<4.0725 0 0>, rot=<0 -0.0654 0 -0.9979>}
 dyn_shoulder_chain_r_03 {parent=dyn_shoulder_chain_r_02, loc=<2.1806 0 -0.0435>, rot=<0 -0.1521 0 -0.9884>}
 neck_01 {parent=spine_05, loc=<11.3138 0.1529 0>, rot=<0 0 -0.2003 -0.9797>}
 neck_02 {parent=neck_01, loc=<4.5627 0 0>, rot=<0 0 0.0504 -0.9987>}
 head {parent=neck_02, loc=<5.3914 0 0>, rot=<0 0 0.115 -0.9934>}
 dyn_hat_01 {parent=head, loc=<11.4017 -6.9793 0>, rot=<0.3154 -0.6329 0.6329 -0.3154>}
 dyn_hat_02 {parent=dyn_hat_01, loc=<4.6565 -0.0011 -0.0685>, rot=<0 0.0541 0 -0.9985>}
 dyn_hat_03 {parent=dyn_hat_02, loc=<4.9463 0.0011 -0.1878>, rot=<0 0.1366 0 -0.9906>}
 dyn_hat_04 {parent=dyn_hat_03, loc=<6.3445 0.0037 0.0789>, rot=<0 0.0523 0 -0.9986>}
 dyn_shawl_back {parent=spine_05, loc=<1.2287 -13.9421 0>, rot=<0.0284 -0.7065 -0.0284 -0.7065>}
 dyn_tassle_r {parent=spine_04, loc=<-2.6083 13.2945 9.4559>, rot=<0.0294 -0.7065 -0.0294 -0.7065>}
 dyn_tassle_l {parent=spine_04, loc=<-2.6067 13.2939 -9.4581>, rot=<0.0294 -0.7065 -0.0294 -0.7065>}
 thigh_l {parent=pelvis, loc=<-2.5693 -0.0109 -7.7501>, rot=<0.0012 0.0572 0.0155 -0.9982>}
 calf_l {parent=thigh_l, loc=<-39.0484 0 0>, rot=<0.0006 -0.0161 -0.0391 -0.9991>}
 calf_twist_01_l {parent=calf_l, loc=<-35.8131 -0.0967 0.0397>, rot=<0.0237 0.0095 -0.0035 -0.9997>}
 calf_twist_02_l {parent=calf_l, loc=<-18.8064 -0.0967 0.0397>, rot=<0.0237 0.0095 -0.0035 -0.9997>}
 foot_l {parent=calf_l, loc=<-42.561 0.1077 0.4292>, rot=<0.0418 -0.0455 0.0343 -0.9975>}
 ball_l {parent=foot_l, loc=<-6.1599 13.8264 0.6806>, rot=<0.0001 -0.0001 -0.7063 -0.7079>}
 thigh_twist_01_l {parent=thigh_l, loc=<-21.6303 0 0>, rot=<0.0013 0 0 -1>}
 dyn_jacket_leg_master_l {parent=thigh_l, loc=<-21.6303 0 0>, rot=<0.0013 0 0 -1>}
 dyn_jacket_side_01_l {parent=dyn_jacket_leg_master_l, loc=<21.3937 -1.3174 -7.4002>, rot=<-0.0755 0.6742 0.7309 0.0741>}
 dyn_jacket_side_02_l {parent=dyn_jacket_side_01_l, loc=<11.8282 0 0>, rot=<0 0 -0.0412 -0.9992>}
 dyn_jacket_side_03_l {parent=dyn_jacket_side_02_l, loc=<13.3963 0 0>, rot=<0 0 -0.0294 -0.9996>}
 dyn_jacket_side_04_l {parent=dyn_jacket_side_03_l, loc=<14.3097 0 0>, rot=<0 0 0 1>}
 dyn_jacket_front_01_l {parent=dyn_jacket_leg_master_l, loc=<21.5718 9.0665 -3.8123>, rot=<-0.0099 0.2138 0.9737 0.0786>}
 dyn_jacket_front_02_l {parent=dyn_jacket_front_01_l, loc=<11.8972 0 0>, rot=<0 0 -0.0412 -0.9992>}
 dyn_jacket_front_03_l {parent=dyn_jacket_front_02_l, loc=<14.4425 0 0>, rot=<0 0 -0.0294 -0.9996>}
 dyn_jacket_front_04_l {parent=dyn_jacket_front_03_l, loc=<15.9799 0 0>, rot=<0 0 0 1>}
 dyn_jacket_front_05_l {parent=dyn_jacket_front_04_l, loc=<10.5896 0 0>, rot=<0 0 0 1>}
 thigh_r {parent=pelvis, loc=<-2.5691 -0.0109 7.7501>, rot=<0.0572 -0.0012 0.9982 0.0155>}
 calf_r {parent=thigh_r, loc=<39.0487 0 0>, rot=<0.0006 -0.0161 -0.0391 -0.9991>}
 calf_twist_01_r {parent=calf_r, loc=<35.8132 0.0967 -0.0397>, rot=<0.0237 0.0095 -0.0035 -0.9997>}
 calf_twist_02_r {parent=calf_r, loc=<18.8064 0.0967 -0.0397>, rot=<0.0237 0.0095 -0.0035 -0.9997>}
 foot_r {parent=calf_r, loc=<42.5611 -0.1077 -0.4292>, rot=<0.0418 -0.0455 0.0343 -0.9975>}
 ball_r {parent=foot_r, loc=<6.1598 -13.8264 -0.6806>, rot=<0.0001 -0.0001 -0.7063 -0.7079>}
 thigh_twist_01_r {parent=thigh_r, loc=<21.6303 0 0>, rot=<0.0013 0 0 -1>}
 dyn_jacket_leg_master_r {parent=thigh_r, loc=<21.6303 0 0>, rot=<0.0013 0 0 -1>}
 dyn_jacket_side_01_r {parent=dyn_jacket_leg_master_r, loc=<-21.3935 1.3174 7.4002>, rot=<0.6742 0.0755 0.0741 -0.7309>}
 dyn_jacket_side_02_r {parent=dyn_jacket_side_01_r, loc=<11.8282 0 0>, rot=<0 0 -0.0412 -0.9992>}
 dyn_jacket_side_03_r {parent=dyn_jacket_side_02_r, loc=<13.3963 0 0>, rot=<0 0 -0.0294 -0.9996>}
 dyn_jacket_side_04_r {parent=dyn_jacket_side_03_r, loc=<14.3097 0 0>, rot=<0 0 0 1>}
 dyn_jacket_front_01_r {parent=dyn_jacket_leg_master_r, loc=<-21.5716 -9.0665 3.8123>, rot=<0.2138 0.0099 0.0786 -0.9737>}
 dyn_jacket_front_02_r {parent=dyn_jacket_front_01_r, loc=<11.8972 0 0>, rot=<0 0 -0.0412 -0.9992>}
 dyn_jacket_front_03_r {parent=dyn_jacket_front_02_r, loc=<14.4425 0 0>, rot=<0 0 -0.0294 -0.9996>}
 dyn_jacket_front_04_r {parent=dyn_jacket_front_03_r, loc=<15.9798 0 0>, rot=<0 0 0 -1>}
 dyn_jacket_front_05_r {parent=dyn_jacket_front_04_r, loc=<10.5896 0 0>, rot=<0 0 0 -1>}
 wolf_thigh_l {parent=pelvis, loc=<-2.5693 -0.0109 -7.7501>, rot=<0.0012 0.0572 0.0155 -0.9982>}
 wolf_calf_l {parent=wolf_thigh_l, loc=<-36.3283 2.1181 0.0962>, rot=<0.0006 -0.0161 -0.0391 -0.9991>}
 wolf_calf_twist_01_l {parent=wolf_calf_l, loc=<-32.6056 -5.0853 0.7725>, rot=<0.0237 0.0095 -0.0035 -0.9997>}
 wolf_calf_twist_02_l {parent=wolf_calf_l, loc=<-10.4852 -4.0769 0.3061>, rot=<0.0237 0.0095 -0.0035 -0.9997>}
 wolf_foot_l {parent=wolf_calf_l, loc=<-23.7181 -7.4961 0.2405>, rot=<0.0418 -0.0455 0.0343 -0.9975>}
 wolf_ball_l {parent=wolf_foot_l, loc=<-19.3349 3.4907 -0.8786>, rot=<0.0001 -0.0001 -0.7063 -0.7079>}
 wolf_toe_l {parent=wolf_ball_l, loc=<8.8736 8.7423 -1.1808>, rot=<0 0 0 -1>}
 toe_claw_l {parent=wolf_ball_l, loc=<0.001 2.6176 3.0483>, rot=<-0.5229 -0.4735 0.4801 -0.5214>}
 wolf_thigh_twist_01_l {parent=wolf_thigh_l, loc=<-21.6303 0 0>, rot=<0.0013 0 0 -1>}
 wolf_thigh_r {parent=pelvis, loc=<-2.5694 -0.0109 7.7501>, rot=<0.0572 -0.0012 0.9982 0.0155>}
 wolf_calf_r {parent=wolf_thigh_r, loc=<36.3283 -2.1181 -0.0962>, rot=<0.0006 -0.0161 -0.0391 -0.9991>}
 wolf_calf_twist_01_r {parent=wolf_calf_r, loc=<32.6056 5.0853 -0.7724>, rot=<0.0237 0.0095 -0.0035 -0.9997>}
 wolf_calf_twist_02_r {parent=wolf_calf_r, loc=<10.4852 4.0769 -0.3061>, rot=<0.0237 0.0095 -0.0035 -0.9997>}
 wolf_foot_r {parent=wolf_calf_r, loc=<23.7181 7.4961 -0.2404>, rot=<0.0418 -0.0455 0.0343 -0.9975>}
 wolf_ball_r {parent=wolf_foot_r, loc=<19.3349 -3.4907 0.8786>, rot=<0.0001 -0.0001 -0.7063 -0.7079>}
 wolf_toe_r {parent=wolf_ball_r, loc=<-8.8736 -8.7424 1.1809>, rot=<0 0 0 -1>}
 toe_claw_r {parent=wolf_ball_r, loc=<-0.001 -2.6176 -3.0482>, rot=<-0.5229 -0.4735 0.4801 -0.5214>}
 wolf_thigh_twist_01_r {parent=wolf_thigh_r, loc=<21.6303 0 0>, rot=<0.0013 0 0 -1>}
 dyn_jacket_chain_r {parent=pelvis, loc=<-1.3121 -4.6998 10.3242>, rot=<0.2325 -0.3739 0.894 -0.083>}
 dyn_Ring_Chain_L {parent=pelvis, loc=<-22.8571 -12.3343 19.0644>, rot=<-0.0127 -0.707 0.0127 -0.707>}
 dyn_Ring_Chain_R {parent=pelvis, loc=<-22.8571 -12.3343 -19.0644>, rot=<-0.0127 -0.707 0.0127 -0.707>}
 dyn_jacket_rear_l_MSTR {parent=pelvis, loc=<-2.165 -10.8867 -8.9693>, rot=<-0.0178 -0.9988 0.0008 -0.0465>}
 dyn_jacket_rear_01_l {parent=dyn_jacket_rear_l_MSTR, loc=<0 0 0>, rot=<0 0 0 1>}
 dyn_jacket_rear_02_l {parent=dyn_jacket_rear_01_l, loc=<10.9389 0 0>, rot=<0 0 0.0419 -0.9991>}
 dyn_jacket_rear_03_l {parent=dyn_jacket_rear_02_l, loc=<14.7838 0 0>, rot=<0 0 0 -1>}
 dyn_jacket_rear_04_l {parent=dyn_jacket_rear_03_l, loc=<14.0431 0 0>, rot=<0 0 0 -1>}
 dyn_jacket_rear_r_MSTR {parent=pelvis, loc=<-2.165 -10.8867 8.9693>, rot=<-0.0178 -0.9988 -0.0008 0.0465>}
 dyn_jacket_rear_01_r {parent=dyn_jacket_rear_r_MSTR, loc=<0 0 0>, rot=<0 0 0 1>}
 dyn_jacket_rear_02_r {parent=dyn_jacket_rear_01_r, loc=<10.9389 0 0>, rot=<0 0 0.0419 -0.9991>}
 dyn_jacket_rear_03_r {parent=dyn_jacket_rear_02_r, loc=<14.7838 0 0>, rot=<0 0 0 -1>}
 dyn_jacket_rear_04_r {parent=dyn_jacket_rear_03_r, loc=<14.0431 0 0>, rot=<0 0 0 -1>}
 ik_hand_root {parent=root, loc=<0 0 0>, rot=<0 0 0 1>}
 ik_hand_gun {parent=ik_hand_root, loc=<-33.8029 3.8283 90.5694>, rot=<-0.4834 -0.3223 -0.5887 0.5621>}
 ik_hand_l {parent=ik_hand_gun, loc=<6.7067 65.8041 13.978>, rot=<-0.0992 -0.9733 -0.2068 0>}
 ik_hand_r {parent=ik_hand_gun, loc=<0 0 0>, rot=<0 0 0 -1>}
 ik_foot_root {parent=root, loc=<0 0 0>, rot=<0 0 0 1>}
 ik_foot_l {parent=ik_foot_root, loc=<16.4438 3.5592 3.2853>, rot=<-0.5 -0.5 0.5 0.5>}
 ik_foot_r {parent=ik_foot_root, loc=<-16.4438 3.5592 3.2853>, rot=<-0.5 0.5 0.5 -0.5>}
]
