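# Secondary-motion Anim Blueprint for the Krampus player body, layered on the pose copied from the parent mesh.
# Jacket uses RigidBody physics, falling back to calf-driven rotation constraints when physics is disabled or LOD exceeds 1.
# AnimDynamics sleeves, weighted ring-chain constraints and AnimDynamics tassels follow, ending in a Control Rig.
# Event graph sets RigidBody gravity from a deprecated helper, then overrides it by skydiving, parachute and vertical-velocity state unless swim-sprinting.
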AnimBlueprint /BRCosmetics/Characters/Player/Male/Medium/Bodies/M_MED_Krampus_01/Meshes/M_MED_Krampus_AnimBP
target_skeleton=&/M_MED_Krampus_Body_01_Skeleton
graphs [
 AnimGraph {nodes [
   n1 = RigidBody!(
    ComponentPose=Local To Component(LocalPose=Use cached pose 'CPFM Cache'())
    Alpha
    OverrideWorldGravity:Vector=Get GravityOverride()
    ExternalForce=Break Fort Anim Input Skydiving External Force(FortAnimInput_SkydivingExternalForce=Get SkydivingExternalForceInput()) -> (
     bUseSkydivingVectorForce=~
     SkydivingVectorForce=~
     HeadToPelvisDirection=~
     FloatingMultiplier=~
     FloatingAdditive=~
     DivingMultiplier=~
     DivingAdditive=~
     ParachutingMultiplier=~
     ParachutingAdditive=~
    ).SkydivingVectorForce
   ) -> (Pose) {
    bOverrideWorldGravity=true
    overridePhysicsAsset=&/M_MED_Krampus_Body_Perf_new_Physics1
    simulationSpace=BaseBoneSpace
    alphaBoolBlend {blendOption=Linear}
    baseBoneRef {boneName=root}
    componentAppliedLinearAccClamp=<1000 1000 1000>
    componentLinearAccScale=<0.4 0.3 0.3>
    componentLinearVelScale=<0.4 0.3 0.3>
    overrideWorldGravity {z=-980}
   }
   n2 = Component To Local(ComponentPose=@n1)
   n3 = "Constraint - dyn_jacket_front_04_l (R)"(ComponentPose=Local To Component(LocalPose=Use cached pose 'CPFM Cache'()), Alpha=1, ConstraintWeights_0=1) {
    alphaBoolBlend {blendOption=Linear}
    boneToModify {boneName=dyn_jacket_front_04_l}
    constraintSetup [{transformType=Rotation, targetBone {boneName=calf_l}}]
    constraintWeights [1]
   }
   n4 = "Constraint - dyn_jacket_front_04_r (R)"(ComponentPose=@n3, Alpha=1, ConstraintWeights_0=1) {
    alphaBoolBlend {blendOption=Linear}
    boneToModify {boneName=dyn_jacket_front_04_r}
    constraintSetup [{transformType=Rotation, targetBone {boneName=calf_r}}]
    constraintWeights [1]
   }
   n5 = "Constraint - dyn_jacket_rear_03_l (R)"(ComponentPose=@n4, Alpha=1, ConstraintWeights_0=1) {
    alphaBoolBlend {blendOption=Linear}
    boneToModify {boneName=dyn_jacket_rear_03_l}
    constraintSetup [{transformType=Rotation, targetBone {boneName=calf_l}}]
    constraintWeights [1]
   }
   n6 = "Constraint - dyn_jacket_rear_03_r (R)"(ComponentPose=@n5, Alpha=1, ConstraintWeights_0=1) {
    alphaBoolBlend {blendOption=Linear}
    boneToModify {boneName=dyn_jacket_rear_03_r}
    constraintSetup [{transformType=Rotation, targetBone {boneName=calf_r}}]
    constraintWeights [1]
   }
   n7 = "Constraint - dyn_jacket_side_03_l (R)"(ComponentPose=@n6, Alpha=1, ConstraintWeights_0=1) {
    alphaBoolBlend {blendOption=Linear}
    boneToModify {boneName=dyn_jacket_side_03_l}
    constraintSetup [{transformType=Rotation, targetBone {boneName=calf_l}}]
    constraintWeights [1]
   }
   n8 = "Constraint - dyn_jacket_side_03_r (R)"(ComponentPose=@n7, Alpha=1, ConstraintWeights_0=1) {
    alphaBoolBlend {blendOption=Linear}
    boneToModify {boneName=dyn_jacket_side_03_r}
    constraintSetup [{transformType=Rotation, targetBone {boneName=calf_r}}]
    constraintWeights [1]
   }
   n9 = Component To Local(ComponentPose=@n8)
   n10 = OR Boolean(A=NOT Boolean(A=Get bRigidBodyEnabled()), B="integer > integer"(A=Get CurrentLOD(), B=1) {is=title K2Node_CallFunction})
   n11 = Anim Dynamics - Bone: dyn_shoulder_sleeve_02_l(
    ComponentPose=Local To Component(LocalPose=Blend Poses by bool(BlendPose_0=@n9, BlendPose_1=@n2, bActiveValue=@n10) {blendType=Linear})
    Alpha=0.5
    ExternalForce="0,0,0"
   ) {
    angularDampingOverride=0.99
    angularSpringConstant=250
    bAngularSpring=true
    bGravityOverrideInSimSpace=true
    bOverrideAngularDamping=true
    gravityScale=0.35
    lODThreshold=1
    simulationSpace=RootRelative
    windScale=0.5
    alphaBoolBlend {blendOption=Linear}
    boundBone {boneName=dyn_shoulder_sleeve_02_l}
    componentAppliedLinearAccClamp=<1000 1000 1000>
    componentLinearAccScale=<1 1 1>
    componentLinearVelScale=<1 1 1>
    physicsBodyDefinitions [{
      sphereCollisionRadius=0
      boundBone {boneName=dyn_shoulder_sleeve_02_l}
      boxExtents=<10 20 20>
      constraintSetup {angularLimitsMax=<0 0 5>, angularLimitsMin=<0 -60 -5>, angularTarget=<0 0 0>}
      localJointOffset=<10 0 -4>
     }]
   }
   n12 = Anim Dynamics - Bone: dyn_shoulder_sleeve_02_r(ComponentPose=@n11, Alpha=0.5, ExternalForce="0,0,0") {
    angularDampingOverride=0.99
    angularSpringConstant=250
    bAngularSpring=true
    bGravityOverrideInSimSpace=true
    bOverrideAngularDamping=true
    gravityScale=0.35
    lODThreshold=1
    simulationSpace=RootRelative
    windScale=0.5
    alphaBoolBlend {blendOption=Linear}
    boundBone {boneName=dyn_shoulder_sleeve_02_r}
    componentAppliedLinearAccClamp=<1000 1000 1000>
    componentLinearAccScale=<1 1 1>
    componentLinearVelScale=<1 1 1>
    physicsBodyDefinitions [{
      sphereCollisionRadius=0
      boundBone {boneName=dyn_shoulder_sleeve_02_r}
      boxExtents=<10 20 20>
      constraintSetup {angularLimitsMax=<0 0 5>, angularLimitsMin=<0 -60 -5>, angularTarget=<0 0 0>}
      localJointOffset=<-10 0 4>
     }]
   }
   n13 = "Constraint - dyn_Ring_Chain_R (PPPPPPPPPPPP)"!(
    ComponentPose=@n12
    Alpha=1
    ConstraintWeights_0=0.1877
    ConstraintWeights_1:float=0.0036
    ConstraintWeights_2:float=0.0734
    ConstraintWeights_3:float=0.1546
    ConstraintWeights_4:float=0.201
    ConstraintWeights_5:float=0.0609
    ConstraintWeights_6:float=0.0173
    ConstraintWeights_7:float=0.0024
    ConstraintWeights_8:float=0.0958
    ConstraintWeights_9:float=0.199
    ConstraintWeights_10:float=0
    ConstraintWeights_11:float=0.0043
   ) -> (Pose) {
    alphaBoolBlend {blendOption=Linear}
    boneToModify {boneName=dyn_Ring_Chain_R}
    constraintSetup [
     {targetBone {boneName=dyn_jacket_rear_01_l}}
     {targetBone {boneName=dyn_jacket_front_01_l}}
     {targetBone {boneName=dyn_jacket_side_04_l}}
     {targetBone {boneName=dyn_jacket_rear_03_l}}
     {targetBone {boneName=dyn_jacket_side_01_l}}
     {targetBone {boneName=dyn_jacket_side_03_l}}
     {targetBone {boneName=dyn_jacket_rear_04_l}}
     {targetBone {boneName=dyn_jacket_front_04_l}}
     {targetBone {boneName=dyn_jacket_side_02_l}}
     {targetBone {boneName=pelvis}}
     {targetBone {boneName=dyn_jacket_rear_04_r}}
     {targetBone {boneName=dyn_jacket_front_02_l}}
    ]
    constraintWeights [1,1,1,1,1,1,1,1,1,1,1,1]
   }
   n14 = "Constraint - dyn_Ring_Chain_L (PPPPPPPPPPPPPPP)"!(
    ComponentPose=@n13
    Alpha=1
    ConstraintWeights_0=0.1745
    ConstraintWeights_1:float=0.0829
    ConstraintWeights_2:float=0
    ConstraintWeights_3:float=0.0059
    ConstraintWeights_4:float=0.0231
    ConstraintWeights_5:float=0
    ConstraintWeights_6:float=0.2061
    ConstraintWeights_7:float=0.0004
    ConstraintWeights_8:float=0.0845
    ConstraintWeights_9:float=0
    ConstraintWeights_10:float=0.0051
    ConstraintWeights_11:float=0.1915
    ConstraintWeights_12:float=0.0061
    ConstraintWeights_13:float=0.1924
    ConstraintWeights_14:float=0.0275
   ) -> (Pose) {
    alphaBoolBlend {blendOption=Linear}
    boneToModify {boneName=dyn_Ring_Chain_L}
    constraintSetup [
     {targetBone {boneName=dyn_jacket_rear_03_r}}
     {targetBone {boneName=dyn_jacket_side_02_r}}
     {targetBone {boneName=dyn_jacket_side_04_l}}
     {targetBone {boneName=dyn_jacket_front_04_r}}
     {targetBone {boneName=dyn_jacket_rear_04_r}}
     {targetBone {boneName=dyn_jacket_side_01_l}}
     {targetBone {boneName=dyn_jacket_side_01_r}}
     {targetBone {boneName=dyn_jacket_front_05_r}}
     {targetBone {boneName=dyn_jacket_side_04_r}}
     {targetBone {boneName=dyn_jacket_front_04_l}}
     {targetBone {boneName=dyn_jacket_front_01_r}}
     {targetBone {boneName=dyn_jacket_rear_01_r}}
     {targetBone {boneName=dyn_jacket_front_03_r}}
     {targetBone {boneName=pelvis}}
     {targetBone {boneName=dyn_jacket_side_03_r}}
    ]
    constraintWeights [1,1,1,1,1,1,1,1,1,1,1,1,1,1,1]
   }
   n15 = Anim Dynamics - Bone: dyn_tassle_l(ComponentPose=@n14, Alpha=0.35, ExternalForce="0,0,0") {
    angularDampingOverride=0.98
    bGravityOverrideInSimSpace=true
    bOverrideAngularDamping=true
    gravityScale=0.5
    lODThreshold=1
    numSolverIterationsPostUpdate=4
    numSolverIterationsPreUpdate=16
    alphaBoolBlend {blendOption=Linear}
    boundBone {boneName=dyn_tassle_l}
    componentLinearAccScale=<1 1 1>
    componentLinearVelScale=<1 1 1>
    physicsBodyDefinitions [{
      sphereCollisionRadius=0
      boundBone {boneName=dyn_tassle_l}
      boxExtents=<15 15 3>
      constraintSetup {angularLimitsMax=<89 89 0>, angularLimitsMin=<-89 -89 0>, angularTarget=<0 0 0>}
      localJointOffset=<0 0 -15>
     }]
    relativeSpaceBone {boneName=pelvis}
   }
   n16 = Anim Dynamics - Bone: dyn_tassle_r(ComponentPose=@n15, Alpha=0.35, ExternalForce="0,0,0") {
    angularDampingOverride=0.98
    bGravityOverrideInSimSpace=true
    bOverrideAngularDamping=true
    gravityScale=0.5
    lODThreshold=1
    numSolverIterationsPostUpdate=4
    numSolverIterationsPreUpdate=16
    alphaBoolBlend {blendOption=Linear}
    boundBone {boneName=dyn_tassle_r}
    componentLinearAccScale=<1 1 1>
    componentLinearVelScale=<1 1 1>
    physicsBodyDefinitions [{
      sphereCollisionRadius=0
      boundBone {boneName=dyn_tassle_r}
      boxExtents=<15 15 3>
      constraintSetup {angularLimitsMax=<89 89 0>, angularLimitsMin=<-89 -89 0>, angularTarget=<0 0 0>}
      localJointOffset=<0 0 -15>
     }]
    relativeSpaceBone {boneName=pelvis}
   }
   n17 = Control Rig(Source=Component To Local(ComponentPose=@n16)) {
    alphaBoolBlend {blendOption=Linear}
    controlRigAssetReference {blueprintRigClass=&/M_MED_Krampus_Body_ControlRig.M_MED_Krampus_Body_ControlRig_C}
   }
   Output Pose(Result=@n17) {layerGroup=None}
   Save cached pose 'CPFM Cache'(Pose=Copy Pose From Mesh() {bUseAttachedParent=true}) {cacheName=CPFM Cache}
  ]}
 EventGraph {nodes [
   Event BlueprintUpdateAnimation()
   > n1 = Sequence!(execute) -> (then_0, then_1, then_2:Exec)
   > n2 = Gravity Override State Function DEPRECATED(CharacterPartAnimInstance=Self-Reference(), GameTimeInSeconds=GetGameTimeInSeconds(), GravityOverrideParams=Get Gravity Override Params())
   > Set GravityOverride(GravityOverride=@n2)
   n3 = Branch(execute=@n1.then_1, Condition=Get bIsSwimSprinting() -> (bIsSwimSprinting:bool))
   n4 = Branch(execute=@n3.else, Condition=Get bIsSkyDiving())
   > n5 = Branch(Condition=Get bIsParachuteOpened())
   Set GravityOverride(execute=[@n4.else, @n5.then], GravityOverride="0, 0,-980")
   n6 = Branch(execute=@n5.else, Condition="float > float"(A=Get PawnVelocityZ(), B=50) {is=title K2Node_CallFunction})
   > Set GravityOverride(GravityOverride="0, 0, 0")
   Set GravityOverride(execute=@n6.else, GravityOverride="0, 0,-300")
   n7 = Get CurrentLOD()
   n8 = Get bRigidBodyEnabled()
   Set LOD(execute=@n1.then_2, LOD:bool=OR Boolean(A=NOT Boolean(A=@n8), B="integer > integer"(A=@n7, B=1))) -> (Output_Get:bool)
  ]}
]
variables [
 Gravity Override Params {type=GravityOverrideParamsStruct}
 GravityOverride {type=Vector}
 LOD {type=boolean}
]
