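# Physics asset for the beard of the Krampus medium male player character cosmetic.
# It has two capsule bodies, a small one on the dyn_beard bone and a large one on spine_05.
# A single constraint links dyn_beard to spine_05 with angular swing and twist limit values, tighter on the first swing axis than on the second.
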PhysicsAsset /BRCosmetics/Characters/Player/Male/Medium/Bodies/M_MED_Krampus_01/Meshes/M_MED_Krampus_Beard_Physics
bodies {
 dyn_beard [{shapeType=Capsule, length=2.7515, radius=1.8391, center=<-0.0212 2.4918 -6.658>, rotation=<rot -23.6947 -90 0.5943>}]
 spine_05 [{shapeType=Capsule, length=14.9439, radius=11.9913, center=<5.1811 -0.9367 1.0021>, rotation=<rot 13.4469 -90 -1.368>}]
}
constraints [{bone1Name=dyn_beard, bone2Name=spine_05, swing1LimitDegrees=5, swing2LimitDegrees=20, twistLimitDegrees=45}]
