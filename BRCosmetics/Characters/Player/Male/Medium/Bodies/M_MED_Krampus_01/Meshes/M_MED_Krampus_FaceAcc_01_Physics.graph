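# Physics asset for the face accessory of the male medium Krampus player cosmetic, giving its hat bones physics bodies.
# Four dyn_hat capsules form a chain from the head, mostly growing toward the last bone, with angular limits between links that allow little twist.
# Head and two upper-spine capsules supply body volumes, with the head as a zero-length capsule.
# The hat root, head and spine_05 constraints leave swing and twist free.
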PhysicsAsset /BRCosmetics/Characters/Player/Male/Medium/Bodies/M_MED_Krampus_01/Meshes/M_MED_Krampus_FaceAcc_01_Physics
bodies {
 dyn_hat_01 [{shapeType=Capsule, length=0.505, radius=0.505, shapeName=dyn_hat_01_capsule, rotation=<rot 0 -90 0>}]
 dyn_hat_02 [{shapeType=Capsule, length=0.5771, radius=2.2315, shapeName=dyn_hat_02_capsule, center=<2.5 0 0>, rotation=<rot 90 0 0>}]
 dyn_hat_03 [{shapeType=Capsule, length=1.7523, radius=2.1063, shapeName=dyn_hat_03_capsule, center=<3.25 0 0>, rotation=<rot 90 0 0>}]
 dyn_hat_04 [{shapeType=Capsule, length=3.6526, radius=3.1111, shapeName=dyn_hat_04_capsule, center=<5.5 0 0>, rotation=<rot 90 0 0>}]
 head [{shapeType=Capsule, length=0, radius=10}]
 spine_04 [{shapeType=Capsule, length=12.9439, radius=13.9913, shapeName=spine_04_capsule, center=<11.5313 0 0>}]
 spine_05 [{shapeType=Capsule, length=14.9439, radius=11.9913, center=<10.5126 0.6208 0.7606>, rotation=<rot 13.447 -90 -1.368>}]
}
constraints [
 {
  bone1Name=dyn_hat_01
  bone2Name=head
  swing1LimitDegrees=45
  swing1Motion=Free
  swing2LimitDegrees=45
  swing2Motion=Free
  twistLimitDegrees=45
  twistMotion=Free
 }
 {bone1Name=dyn_hat_02, bone2Name=dyn_hat_01, swing1LimitDegrees=22.5, swing2LimitDegrees=45, twistLimitDegrees=5}
 {bone1Name=dyn_hat_03, bone2Name=dyn_hat_02, swing1LimitDegrees=22.5, swing2LimitDegrees=45, twistLimitDegrees=5}
 {bone1Name=dyn_hat_04, bone2Name=dyn_hat_03, swing1LimitDegrees=22.5, swing2LimitDegrees=45, twistLimitDegrees=5}
 {
  bone1Name=spine_05
  bone2Name=spine_04
  swing1LimitDegrees=45
  swing1Motion=Free
  swing2LimitDegrees=45
  swing2Motion=Free
  twistLimitDegrees=45
  twistMotion=Free
 }
 {
  bone1Name=head
  bone2Name=spine_04
  swing1LimitDegrees=45
  swing1Motion=Free
  swing2LimitDegrees=45
  swing2Motion=Free
  twistLimitDegrees=45
  twistMotion=Free
 }
]
