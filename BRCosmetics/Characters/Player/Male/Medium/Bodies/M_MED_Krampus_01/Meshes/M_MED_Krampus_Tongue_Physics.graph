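# Physics asset for the tongue of the male medium Krampus character cosmetic, letting it hang and swing as a jointed chain.
# Five small capsule bodies on dynamic tongue bones link in sequence from the tongue root through identical constraints with moderate swing and tight twist limits.
# Larger head and upper-spine capsules have no constraints, apparently serving as collision volumes for the chain.
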PhysicsAsset /BRCosmetics/Characters/Player/Male/Medium/Bodies/M_MED_Krampus_01/Meshes/M_MED_Krampus_Tongue_Physics
bodies {
 dyn_tongue_01 [{shapeType=Capsule, length=0, radius=2}]
 dyn_tongue_02 [{shapeType=Capsule, length=0, radius=2}]
 dyn_tongue_03 [{shapeType=Capsule, length=0, radius=2}]
 dyn_tongue_04 [{shapeType=Capsule, length=0, radius=2}]
 dyn_tongue_05 [{shapeType=Capsule, length=0, radius=2}]
 head [{shapeType=Capsule, length=0, radius=10}]
 spine_05 [{shapeType=Capsule, length=14.9439, radius=11.9913, center=<10.5126 0.6208 0.7606>, rotation=<rot 13.447 -90 -1.368>}]
 tongue [{shapeType=Capsule, length=0, radius=2.5}]
}
constraints [
 {bone1Name=dyn_tongue_01, bone2Name=tongue, swing1LimitDegrees=22.5, swing2LimitDegrees=22.5, twistLimitDegrees=5}
 {bone1Name=dyn_tongue_02, bone2Name=dyn_tongue_01, swing1LimitDegrees=22.5, swing2LimitDegrees=22.5, twistLimitDegrees=5}
 {bone1Name=dyn_tongue_03, bone2Name=dyn_tongue_02, swing1LimitDegrees=22.5, swing2LimitDegrees=22.5, twistLimitDegrees=5}
 {bone1Name=dyn_tongue_04, bone2Name=dyn_tongue_03, swing1LimitDegrees=22.5, swing2LimitDegrees=22.5, twistLimitDegrees=5}
 {bone1Name=dyn_tongue_05, bone2Name=dyn_tongue_04, swing1LimitDegrees=22.5, swing2LimitDegrees=22.5, twistLimitDegrees=5}
]
