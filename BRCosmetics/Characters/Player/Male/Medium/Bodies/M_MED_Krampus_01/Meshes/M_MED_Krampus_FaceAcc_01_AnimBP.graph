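# Animation Blueprint for the Krampus face accessory, copying the attached parent mesh's pose and adding secondary motion to parts such as beard, tongue and hat.
# The physics path applies a hat bone offset, then chains three RigidBody nodes with separate physics assets and a variable tongue alpha.
# bUseRigidBody selects that path or a baked pose layered onto the hat and tongue branches.
# A final alpha-driven bone scale can shrink the tongue to near zero.
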AnimBlueprint /BRCosmetics/Characters/Player/Male/Medium/Bodies/M_MED_Krampus_01/Meshes/M_MED_Krampus_FaceAcc_01_AnimBP
target_skeleton=&/M_MED_Krampus_FaceAcc_01_Skeleton
graphs [
 AnimGraph {nodes [
   n1 = "Transform (Modify) Bone - Bone: dyn_hat_01"!(
    ComponentPose=Local To Component(LocalPose=Use cached pose 'CPFM Cache'())
    bAlphaBoolEnabled:bool=True
    Translation="0,0,0"
    Rotation="0.000,0.000,30"
    Scale="1,1,1"
   ) -> (Pose) {
    alphaInputType=Bool
    bAlphaBoolEnabled=false
    rotationMode=BMM_Additive
    translationMode=BMM_Additive
    translationSpace=BCS_ParentBoneSpace
    alphaBoolBlend {blendInTime=0.1, blendOption=Linear, blendOutTime=0.1}
    boneToModify {boneName=dyn_hat_01}
    rotation {roll=30}
   }
   n2 = RigidBody(ComponentPose=@n1) {
    bOverrideWorldGravity=true
    overridePhysicsAsset=&/M_MED_Krampus_Beard_Physics
    simulationSpace=BaseBoneSpace
    alphaBoolBlend {blendOption=Linear}
    baseBoneRef {boneName=head}
    componentAppliedLinearAccClamp=<2000 2000 2000>
    componentLinearAccScale=<1 1 1>
    componentLinearVelScale=<1 1 1>
    overrideWorldGravity {z=-980}
   }
   n3 = RigidBody(ComponentPose=@n2, Alpha=Get TonguePhysicsAlpha() -> (TonguePhysicsAlpha:float)) {
    bOverrideWorldGravity=true
    overridePhysicsAsset=&/M_MED_Krampus_Tongue_Physics
    alphaBoolBlend {blendOption=Linear}
    componentAppliedLinearAccClamp=<2000 2000 2000>
    componentLinearAccScale=<1 1 1>
    componentLinearVelScale=<1 1 1>
    overrideWorldGravity {z=-980}
   }
   n4 = RigidBody(ComponentPose=@n3) {
    bOverrideWorldGravity=true
    overridePhysicsAsset=&/M_MED_Krampus_FaceAcc_01_Physics
    simulationSpace=BaseBoneSpace
    alphaBoolBlend {blendOption=Linear}
    baseBoneRef {boneName=neck_01}
    componentLinearAccScale=<0.3 0.3 0.3>
    componentLinearVelScale=<0.3 0.3 0.3>
    overrideWorldGravity {z=-540}
   }
   n5 = Component To Local(ComponentPose=@n4)
   n6 = Layered blend per bone(
    BasePose=Use cached pose 'CPFM Cache'()
    BlendPoses_0=Sequence Player 'M_MED_Krampus_FaceAcc_01_BakePose'() {sequence=&/M_MED_Krampus_FaceAcc_01_BakePose}
   ) {layerSetup [{branchFilters [{blendDepth=1, boneName=dyn_hat_01}, {blendDepth=1, boneName=tongue}]}]}
   n7 = Get bUseRigidBody() -> (bUseRigidBody:bool)
   n8 = "Transform (Modify) Bone - Bone: dyn_tongue_01"!(
    ComponentPose=Local To Component(LocalPose=Blend Poses by bool(BlendPose_0=@n5, BlendPose_1=@n6, bActiveValue=@n7) {blendType=Linear})
    Alpha=Get TongueBoneAlpha() -> (TongueBoneAlpha:float)
    Scale="0.001,0.001,0.001"
   ) -> (Pose) {scaleMode=BMM_Replace, alphaBoolBlend {blendOption=Linear}, boneToModify {boneName=dyn_tongue_01}, scale=<0.001 0.001 0.001>}
   Output Pose(Result=Component To Local(ComponentPose=@n8)) {layerGroup=None}
   Save cached pose 'CPFM Cache'(Pose=Copy Pose From Mesh() {bUseAttachedParent=true}) {cacheName=CPFM Cache}
  ]}
 EventGraph {links [], nodes []}
]
variables []
